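# Physics asset for the Majesty Taco backpack cosmetic's skeletal mesh, giving its dangling parts secondary physical motion.
# Capsule bodies cover the pack and its dynamic handle, button, crown and taco bones.
# The handle and button constrain directly to the pack with fairly tight limits, while a four-link taco chain and two-link crown chain hang from it with wider swing and limited twist.
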PhysicsAsset /BRCosmetics/Accessories/FORT_Backpacks/M_MED_Majesty_Taco/Meshes/M_MED_Majesty_Taco_Pack_Physics
bodies {
 dyn_button [{shapeType=Capsule, length=0, radius=3, center=<1.5 0 0>}]
 dyn_crown_01 [{shapeType=Capsule, length=2.8931, radius=0.5, center=<3.7925 -0.2332 0.0001>, rotation=<rot 90 0 0>}]
 dyn_crown_02 [{shapeType=Capsule, length=6.5697, radius=0.7376, center=<4.2917 -0.6854 0.0001>, rotation=<rot 0.8907 179.9999 -179.9999>}]
 dyn_handle [{shapeType=Capsule, length=8, radius=2, center=<0.6667 0 0>}]
 dyn_taco_01 [{shapeType=Capsule, length=1.4572, radius=1, center=<1.482 0 0.0001>, rotation=<rot 90 0 0>}]
 dyn_taco_02 [{shapeType=Capsule, length=1.4572, radius=1, center=<1.482 0 0.0001>, rotation=<rot 90 0 0>}]
 dyn_taco_03 [{shapeType=Capsule, length=1.4572, radius=1, center=<1.482 0 0.0001>, rotation=<rot 90 0 0>}]
 dyn_taco_04 [{shapeType=Capsule, length=3.7758, radius=1.9755, center=<3.3495 -0.2617 0.0019>, rotation=<rot 0.6418 -180 -180>}]
 pack [{shapeType=Capsule, length=10.7769, radius=16.3872, center=<0.0499 -0.4689 -6.5245>, rotation=<rot 90 90 86.6371>}]
}
constraints [
 {bone1Name=dyn_handle, bone2Name=pack, swing1LimitDegrees=20, swing2LimitDegrees=15, twistLimitDegrees=10}
 {bone1Name=dyn_button, bone2Name=pack, swing1LimitDegrees=5, swing2LimitDegrees=5, twistLimitDegrees=15}
 {bone1Name=dyn_taco_01, bone2Name=pack, swing1LimitDegrees=45, swing2LimitDegrees=45, twistLimitDegrees=5}
 {bone1Name=dyn_taco_02, bone2Name=dyn_taco_01, swing1LimitDegrees=45, swing2LimitDegrees=45, twistLimitDegrees=5}
 {bone1Name=dyn_taco_03, bone2Name=dyn_taco_02, swing1LimitDegrees=45, swing2LimitDegrees=45, twistLimitDegrees=5}
 {bone1Name=dyn_taco_04, bone2Name=dyn_taco_03, swing1LimitDegrees=45, swing2LimitDegrees=45, twistLimitDegrees=5}
 {bone1Name=dyn_crown_01, bone2Name=pack, swing1LimitDegrees=45, swing2LimitDegrees=45, twistLimitDegrees=25}
 {bone1Name=dyn_crown_02, bone2Name=dyn_crown_01, swing1LimitDegrees=45, swing2LimitDegrees=45, twistLimitDegrees=25}
]
